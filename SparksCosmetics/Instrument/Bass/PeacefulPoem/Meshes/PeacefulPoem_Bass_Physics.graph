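# Physics asset for the PeacefulPoem bass, a Sparks instrument cosmetic, that sets up a dangling cloth bone chain.
# Small capsules cover a master cloth bone and five chained segments, while two capsules on the handle bone include a large offset, tilted body volume.
# Angular-limited constraints link the chain from the handle through the master to the tip, with twist tight at the first segment and looser toward the end.
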PhysicsAsset /SparksCosmetics/Instrument/Bass/PeacefulPoem/Meshes/PeacefulPoem_Bass_Physics
bodies {
 PeacefulPoem_Bass_dyn_cloth_01 [{
   shapeType=Capsule
   length=2.29
   radius=2
   shapeName=PeacefulPoem_Bass_dyn_cloth_01_capsule
   center=<2.3079 0.1911 0.1123>
   rotation=<rot 90 0 0>
  }]
 PeacefulPoem_Bass_dyn_cloth_02 [{
   shapeType=Capsule
   length=4.2376
   radius=2
   shapeName=PeacefulPoem_Bass_dyn_cloth_02_capsule
   center=<3.3221 -0.0326 -0.2718>
   rotation=<rot 90 0 0>
  }]
 PeacefulPoem_Bass_dyn_cloth_03 [{
   shapeType=Capsule
   length=4.408
   radius=2
   shapeName=PeacefulPoem_Bass_dyn_cloth_03_capsule
   center=<3.0948 -0.0816 0.0375>
   rotation=<rot 90 0 0>
  }]
 PeacefulPoem_Bass_dyn_cloth_04 [{
   shapeType=Capsule
   length=4.5022
   radius=2
   shapeName=PeacefulPoem_Bass_dyn_cloth_04_capsule
   center=<3.0796 0.1004 -0.1926>
   rotation=<rot 90 0 0>
  }]
 PeacefulPoem_Bass_dyn_cloth_05 [{
   shapeType=Capsule
   length=2.0112
   radius=2.2033
   shapeName=PeacefulPoem_Bass_dyn_cloth_05_capsule
   center=<2.1336 -0.0876 -0.1467>
   rotation=<rot 90 0 0>
  }]
 PeacefulPoem_Bass_dyn_cloth_MSTR [{shapeType=Capsule, length=0.0685, radius=1.6439, rotation=<rot 90 0 0>}]
 handle [
  {shapeType=Capsule, length=0, radius=3, shapeName=handle_capsule}
  {
   shapeType=Capsule
   length=4.1927
   radius=8.9936
   shapeName=handle_capsule
   center=<-0.6696 -11.8875 -11.5269>
   rotation=<rot 0.0001 0 -20.8361>
  }
 ]
}
constraints [
 {bone1Name=PeacefulPoem_Bass_dyn_cloth_MSTR, bone2Name=handle, swing1LimitDegrees=45, swing2LimitDegrees=45, twistLimitDegrees=45}
 {
  bone1Name=PeacefulPoem_Bass_dyn_cloth_01
  bone2Name=PeacefulPoem_Bass_dyn_cloth_MSTR
  swing1LimitDegrees=35
  swing2LimitDegrees=20
  twistLimitDegrees=5
 }
 {
  bone1Name=PeacefulPoem_Bass_dyn_cloth_02
  bone2Name=PeacefulPoem_Bass_dyn_cloth_01
  swing1LimitDegrees=40
  swing2LimitDegrees=30
  twistLimitDegrees=10
 }
 {
  bone1Name=PeacefulPoem_Bass_dyn_cloth_03
  bone2Name=PeacefulPoem_Bass_dyn_cloth_02
  swing1LimitDegrees=30
  swing2LimitDegrees=40
  twistLimitDegrees=20
 }
 {
  bone1Name=PeacefulPoem_Bass_dyn_cloth_04
  bone2Name=PeacefulPoem_Bass_dyn_cloth_03
  swing1LimitDegrees=35
  swing2LimitDegrees=45
  twistLimitDegrees=45
 }
 {
  bone1Name=PeacefulPoem_Bass_dyn_cloth_05
  bone2Name=PeacefulPoem_Bass_dyn_cloth_04
  swing1LimitDegrees=35
  swing2LimitDegrees=45
  twistLimitDegrees=45
 }
]
